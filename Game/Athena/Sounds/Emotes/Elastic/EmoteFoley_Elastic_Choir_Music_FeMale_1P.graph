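# First-person, female-variant sound cue for the Elastic emote, combining foley, effects and choir music.
# A root mixer starts the foley and one of three randomly chosen wind rushes immediately.
# A fixed short delay then brings in a flash sound and a looping female choir, which an enveloper fades in.
# The cue uses the first-person emote music sound class and concurrency rule, and feeds the music audio-analysis submix.
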SoundCue /Game/Athena/Sounds/Emotes/Elastic/EmoteFoley_Elastic_Choir_Music_FeMale_1P
root=Mixer_0
nodes [
 {id=Delay_0, children [Mixer_1], properties {delayMax=0.8, delayMin=0.8}}
 {
  id=Enveloper_0
  children [WavePlayer_0]
  properties {loopEnd=2, loopStart=2, pitchCurve {editorCurveData {keys [{time=0}]}}, volumeCurve {editorCurveData {keys [<at 0 0 RCIM_Cubic>,<at 1 1>]}}}
 }
 {id=Mixer_0, children [WavePlayer_1,Random_0,Delay_0], properties {inputVolume [1,1,1]}}
 {id=Mixer_1, children [WavePlayer_5,Enveloper_0], properties {inputVolume [1,1]}}
 {id=Random_0, children [WavePlayer_2,WavePlayer_3,WavePlayer_4], properties {weights [1,1,1]}}
 {id=WavePlayer_0, properties {bLooping=true, soundWaveAssetPtr=&/Emote_Elastic_Choir_Loop_Female}}
 {id=WavePlayer_1, properties {soundWaveAssetPtr=&/Emote_Elastic_Foley}}
 {id=WavePlayer_2, properties {soundWaveAssetPtr=&/Emote_Elastic_Wind_Rush_01}}
 {id=WavePlayer_3, properties {soundWaveAssetPtr=&/Emote_Elastic_Wind_Rush_02}}
 {id=WavePlayer_4, properties {soundWaveAssetPtr=&/Emote_Elastic_Wind_Rush_03}}
 {id=WavePlayer_5, properties {soundWaveAssetPtr=&/Emote_Elastic_Flash_01}}
]
properties {
 maxDistance=2097152
 soundClassObject=/Game/Sounds/FortSoundClassesAndModes/Fort_Music_Emote_1P_PSM
 volumeMultiplier=1
 concurrencySet [/Game/Sounds/Concurrency_Rules/Athena_EmoteMusic_Conc]
 soundSubmixSends [{minSendLevel=0, soundSubmix=&Fortnite_AudioAnalysis_Music}]
}
